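# Lit material for Niagara sprites in the Inferno emote, drawing a height chart with numbers, measurement lines and a framed background.
# UV math spaces flipbook numbers vertically, with switches for skipping, side and mirroring.
# Numbers and lines feed a DistanceField function, whose mask blends text and background colours or supplies see-through masking.
# Procedural edge normals bevel the frame, and dynamic parameters drive a PreviousFrameSwitch offset for fake motion blur.
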Material /BRCosmetics/Athena/Effects/Particles/Emotes/NS_Emote_Inferno/Materials/M_Emote_Inferno_Height_Chart
graph {
 properties {
  bUsedWithInstancedStaticMeshes=true
  bUsedWithNiagaraSprites=true
  opacityMaskClipValue=0.01
  pixelDepthOffsetMode=PDOM_Legacy
  usedShadingModels=Default Lit
 }
 nodes [
  n1 = VectorParameter() {parameterName=Chart Text Color, defaultValue=<rgba 0.03 0.03 0.03 0.1>}
  n2 = AppendVector(A=@n1.RGB, B=@n1.RGB)
  n3 = VectorParameter() {parameterName=Chart Background Color, defaultValue=<rgba 0.4 0.4 0.4 0.4>}
  n4 = TextureCoordinate()
  n5 = Constant2Vector() {desc=Particle Size, g=210, r=75}
  n6 = Multiply(A=@n4, B=@n5)
  n7 = ScalarParameter() {defaultValue=2, parameterName=Frame Thickness}
  n8 = Subtract(A=@n6, B=@n7)
  n9 = Ceil(None=@n8)
  n10 = Multiply(A=OneMinus(None=@n4), B=@n5)
  n11 = Subtract(A=@n10, B=@n7)
  n12 = Ceil(None=@n11)
  n13 = MaterialFunctionCall(Float2=Multiply(A=Saturate(None=@n9), B=Saturate(None=@n12))) {materialFunction=&BreakOutFloat2Components}
  n14 = ScalarParameter() {defaultValue=1, group=Numbers And Lines Scale And Offset, parameterName=Numbers_ScaleX}
  n15 = ScalarParameter() {defaultValue=1, group=Numbers And Lines Scale And Offset, parameterName=Numbers_ScaleY}
  n16 = Add(
   A=Multiply(A=TextureCoordinate(), B=AppendVector(A=@n14, B=@n15))
   B=AppendVector(
    A=ScalarParameter() {group=Numbers And Lines Scale And Offset, parameterName=Numbers_OffsetX}
    B=ScalarParameter() {group=Numbers And Lines Scale And Offset, parameterName=Numbers_OffsetY}
   )
  )
  n17 = StaticSwitchParameter(True=@n16, False=@n16) {parameterName=IsSeeThrough}
  n18 = ScalarParameter() {defaultValue=210, parameterName=Particle Size Y}
  n19 = Divide(A=Multiply(A=ComponentMask(None=@n17) {g=true}, B=@n18)) {constB=10, desc=Max space between numbers}
  n20 = ScalarParameter() {defaultValue=8, desc=Max 10cm, parameterName=Text Size Y}
  n21 = Divide(A=@n20) {constA=10, constB=10}
  n22 = Floor(None=Subtract(A=@n19, B=Multiply(A=@n21) {constB=0.5}) {constB=0.5, desc=Offset to line up}) {desc=Getting frame numbers}
  n23 = Divide(A=@n22, B=Constant() {desc=Total number of frames, r=20}) {constB=11, desc=Remapping to 0-1}
  n24 = ComponentMask(None=@n17) {r=true}
  n25 = ConstantBiasScale(None=@n24) {bias=-0.5, scale=2}
  n26 = Divide(A=OneMinus(None=Abs(None=@n25)))
  n27 = StaticSwitchParameter(True=OneMinus(None=@n24), False=@n24) {parameterName=ChangeNumbersSide}
  n28 = ScalarParameter() {defaultValue=75, parameterName=Particle Size X}
  n29 = Subtract(
   A=Multiply(A=StaticSwitchParameter(True=@n26, False=@n27) {parameterName=ShowNumbersOnBothSides}, B=@n28)
   B=ScalarParameter() {defaultValue=4.5, parameterName=Margin Left Side X}
  )
  n30 = ScalarParameter() {defaultValue=10, parameterName=Text Size X}
  n31 = Frac(None=Saturate(None=Divide(A=@n29, B=@n30))) {desc=UV X}
  n32 = OneMinus(None=@n31)
  n33 = Saturate(None=Sign(None=@n25))
  n34 = StaticSwitchParameter(True=LinearInterpolate(A=@n31, B=@n32, Alpha=@n33), False=StaticSwitchParameter(True=@n32, False=@n31) {parameterName=ChangeNumbersSide}) {parameterName=ShowNumbersOnBothSides}
  n35 = Max(A=Subtract(A=@n19) {constB=0.5}) {constB=0}
  n36 = Divide(A=Subtract(A=Frac(None=@n35)) {constB=0.5}, B=@n21)
  n37 = Frac(None=Saturate(None=Add(A=@n36) {constB=0.5})) {desc=UV Y}
  n38 = Multiply(A=TextureCoordinate() {uTiling=0.5, vTiling=0.5}, B=Divide(A=AppendVector(A=@n28, B=@n18), B=AppendVector(A=@n30, B=@n20)))
  n39 = TextureSample!(
   UVs=MaterialFunctionCall("Animation  Phase (0-1)"=@n23, Number of Rows=Constant() {r=5}, Number of Columns=Constant() {r=4}, UVs=AppendVector(A=@n34, B=@n37)) {materialFunction=&FlipBook}.UVs
   Tex
   "DDX(UVs)"=DDX(Value=@n38)
   "DDY(UVs)"=DDY(Value=@n38)
   Apply View MipBias
  ) {
   mipValueMode=TMVM_Derivative
   samplerType=SAMPLERTYPE_LinearGrayscale
   texture=/BRCosmetics/Athena/Effects/Particles/Emotes/NS_Emote_Inferno/Textures/T_Emote_Inferno_Chart
  }
  n40 = OneMinus(None=Saturate(None=Floor(None=@n23))) {desc=Masking numbers not present in the flipbook}
  n41 = Saturate(None=Fmod(A=@n22, B=Constant() {r=2}))
  n42 = OneMinus(None=@n41)
  n43 = LinearInterpolate(A=@n42, B=@n41, Alpha=@n33) {desc=Flip flopping numbers}
  n44 = StaticSwitchParameter(True=@n42, False=@n41) {defaultValue=true, parameterName=FlipSkippedNumbers}
  n45 = StaticSwitchParameter(True=Multiply(A=@n40, B=StaticSwitchParameter(True=@n43, False=@n44) {parameterName=FlipflopSkippedNumbers}), False=@n40) {defaultValue=true, parameterName=SkipEveryOtherNumber}
  n46 = TextureCoordinate()
  n47 = ComponentMask(None=@n46) {g=true}
  n48 = OneMinus(None=Abs(None=ConstantBiasScale(None=@n47) {bias=-0.5, scale=2}))
  n49 = ScalarParameter() {defaultValue=0.0666, parameterName=Lines Margin}
  n50 = Floor(None=Saturate(None=Divide(A=@n48, B=@n49)))
  n51 = ComponentMask(None=@n46) {r=true}
  n52 = Multiply(A=Subtract(A=Multiply(A=@n51) {constB=2}), B=ScalarParameter() {defaultValue=0.9, parameterName=Lines Width})
  n53 = TextureSample!(UVs=AppendVector(A=Add(A=@n52), B=Multiply(A=@n19) {constB=2}), Tex, Apply View MipBias) {
   samplerType=SAMPLERTYPE_LinearGrayscale
   texture=/BRCosmetics/Athena/Effects/Particles/Emotes/NS_Emote_Inferno/Textures/T_Emote_Inferno_Chart_Lines
  }
  n54 = MaterialFunctionCall!(
   DistanceField=Max(A=Multiply(A=@n39.R, B=@n45), B=Multiply(A=@n50, B=@n53.R) {constA=1})
   EdgeSoftness=ScalarParameter() {defaultValue=0.05, parameterName=Edge Softness}
  ) {materialFunction=/Engine/Functions/Engine_MaterialFunctions01/Texturing/DistanceField}
  n55 = Multiply(A=Multiply(A=@n13.R, B=@n13.R), B=OneMinus(None=@n54.Result))
  n56 = StaticSwitchParameter(True=@n2, False=LinearInterpolate(A=@n2, B=AppendVector(A=@n3.RGB, B=@n3.RGB), Alpha=@n55)) {parameterName=IsSeeThrough}
  ComponentMask(None=@n56) {b=true, g=true, r=true}
  ComponentMask(None=@n56) {a=true}
  Constant() {r=1}
  n57 = ConstantBiasScale(None=@n4) {bias=-0.5, scale=2}
  n58 = Multiply(A=OneMinus(None=Abs(None=@n57)), B=Divide(A=Divide(A=@n5, B=@n7)))
  n59 = MaterialFunctionCall(Float2=@n58) {materialFunction=&BreakOutFloat2Components}
  n60 = Saturate(None=Ceil(None=Subtract(A=@n59.R, B=@n59.R)))
  n61 = AppendVector(A=OneMinus(None=@n60), B=@n60)
  n62 = Divide(A=Multiply(A=@n8) {constB=-1}, B=@n7)
  n63 = Sine(None=Saturate(None=Divide(A=@n62)))
  n64 = Divide(A=@n7)
  n65 = Multiply(A=Frac(None=OneMinus(None=@n63)), B=Sign(None=Subtract(A=@n6, B=@n64)))
  n66 = MaterialFunctionCall(Float2=@n58) {materialFunction=&BreakOutFloat2Components}
  n67 = Saturate(None=Ceil(None=Subtract(A=@n66.R, B=@n66.R)))
  n68 = AppendVector(A=OneMinus(None=@n67), B=@n67)
  n69 = Divide(A=Multiply(A=@n11) {constB=-1}, B=@n7)
  n70 = Sine(None=Saturate(None=Divide(A=@n69)))
  n71 = Subtract(A=@n10, B=@n64)
  n72 = Multiply(A=Frac(None=OneMinus(None=@n70)), B=Multiply(A=Sign(None=@n71)) {constB=-1})
  n73 = TextureCoordinate()
  DeriveNormalZ(InXY=LinearInterpolate(A=Multiply(A=@n61, B=@n65), B=Multiply(A=@n68, B=@n72), Alpha=Round(None=@n73)))
  n74 = Saturate(None=TwoSidedSign())
  LinearInterpolate(Alpha=@n74)
  n75 = MaterialFunctionCall(Float2=@n16) {materialFunction=&BreakOutFloat2Components}
  LinearInterpolate(A=AppendVector(A=OneMinus(None=@n75.R), B=@n75.G), B=@n16, Alpha=Saturate(None=TwoSidedSign()))
  Multiply(A=@n74, B=@n55)
  n76 = DynamicParameter() {paramNames [Motion Blur X, Motion Blur Y, "", ""]}
  n77 = VertexTangentWS()
  n78 = Constant3Vector() {constant=<rgba 0 0 1 0>}
  PreviousFrameSwitch(Current Frame=Constant3Vector(), Previous Frame=Add(A=Multiply(A=@n76.Motion Blur X, B=@n77), B=Multiply(A=@n76.Motion Blur Y, B=@n78)))
  StaticSwitchParameter(True=@n54.Result, False=Constant() {r=1}) {parameterName=IsSeeThrough}
 ]
}
