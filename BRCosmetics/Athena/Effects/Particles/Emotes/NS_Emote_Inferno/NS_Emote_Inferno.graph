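# Niagara system for the Inferno emote that stages a height-chart backdrop with two timed camera flashes.
# CameraFlash bursts short-lived particles whose intensity curve drives sprite, point light and component renderers, including a spotlight intensity value.
# HeightChart spawns a local-axis-aligned sprite, with a second renderer and reversed facing vector for the backside.
# Age-indexed curves bounce it in and out, feeding motion-blur-shaped dynamic material parameters, while it outlives its nominal lifetime.
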NiagaraSystem /BRCosmetics/Athena/Effects/Particles/Emotes/NS_Emote_Inferno/NS_Emote_Inferno
dependencies {
 usedDataInterfaces [&NiagaraDataInterfaceCurve, &NiagaraDataInterfaceVector2DCurve]
 usedDynamicInputs []
 usedModules [&SystemState]
}
emitters {
 CameraFlash {
  emitterUpdateScript [
   &EmitterState()
   &SpawnBurst_Instantaneous(Spawn Time=2.466, Use Spawn Probability=false, UseLoopCountLimit=false)
   &SpawnBurst_Instantaneous(Spawn Time=5.21, Use Spawn Probability=false, UseLoopCountLimit=false) {module=SpawnBurst_Instantaneous001}
  ]
  particleSpawnScript [&InitializeParticle(
    Color Mode=::NewEnumerator0
    Lifetime=0.2
    Position Offset=<0 100 135>
    Sprite Size Mode=::NewEnumerator3
    Uniform Sprite Size=133
    UsePositionOffset=true
    Write Material Random=true
   )]
  particleUpdateScript [
   /SetVariables(Particles.FlashIntensity:NiagaraFloat=&FloatFromCurve(FloatCurve=DataInterface() {propertyValues {Curve {keys [
         {arriveTangent=9.0725, time=0, value=0}
         {arriveTangent=-23.1883, leaveTangent=-3.6307, time=0.1, value=1}
         <at 0.2 0.325>
         <at 0.5 0.05>
         {arriveTangent=0.0076, leaveTangent=-0.0405, time=1, value=0}
        ]}}})) {isSetParameters=true}
   &ParticleState()
   /SetVariables(
    Particles.Color=&MakeLinearColorFromVectorAndFloat("Vector (RGB)"=<1 1 1>, "Float (Alpha)"=$StackContext.FlashIntensity)
    Particles.PointLightColor:LinearColor=&MakeLinearColorFromVectorAndFloat("Vector (RGB)"=<1 1 1>, "Float (Alpha)"=&Multiply_Float(A=$StackContext.FlashIntensity, B=100))
    Particles.PointLightPosition:NiagaraPosition=&AddVectorToPosition(Position=$Particles.Position, Vector=<0 -50 0>)
    Particles.SpotLightIntensity:NiagaraFloat=&Multiply_Float(A=$StackContext.FlashIntensity, B=500)
   ) {isSetParameters=true}
  ]
  rendererClasses [
   &NiagaraComponentRendererProperties
   &NiagaraSpriteRendererProperties
   &NiagaraLightRendererProperties
  ]
  renderers [
   {rendererClass=&NiagaraComponentRendererProperties, rendererIndex=0}
   {rendererClass=&NiagaraSpriteRendererProperties, rendererIndex=1}
   {rendererClass=&NiagaraLightRendererProperties, rendererIndex=2}
  ]
 }
 HeightChart {
  emitterUpdateScript [
   &EmitterState()
   &SpawnBurst_Instantaneous(Spawn Time=0.666, Use Spawn Probability=false, UseLoopCountLimit=false)
  ]
  particleSpawnScript [
   &InitializeParticle(
    Color=<rgba 0.75 0.75 0.75 0.5>
    Lifetime=5
    Position Offset=<0 -55 0>
    Sprite Size=<75 210>
    Sprite Size Mode=::NewEnumerator4
    UsePositionOffset=true
    Write Material Random=true
   )
   &SpriteFacingAndAlignment(
    Alignment Coordinate Space=::Local
    Facing Coordinate Space=::Local
    Sprite Alignment=<0 0 1>
    Sprite Facing=<0 1 0>
    WriteAlignment=true
    WriteFacing=true
   )
   /SetVariables(Particles.BacksideSpriteFacing:Vector3f=&Multiply_VectorByFloat(Vector=$Particles.SpriteFacing, Float=-1)) {isSetParameters=true}
  ]
  particleUpdateScript [
   &ParticleState(Kill Particles When Lifetime Has Elapsed=false)
   &ScaleSpriteSize(
    Non-Uniform Curve Index:NiagaraFloat=$Particles.Age
    Non-Uniform Curve Sprite Scale=DataInterface() {propertyValues {
      XCurve {keys [
        {interpMode=RCIM_Cubic, tangentMode=RCTM_Break, time=6, value=1}
        {interpMode=RCIM_Cubic, tangentMode=RCTM_Break, time=6.066, value=0.875}
        {interpMode=RCIM_Cubic, tangentMode=RCTM_Break, time=6.233, value=1.5}
        {arriveTangent=-17.6836, interpMode=RCIM_Cubic, tangentMode=RCTM_Break, time=6.366, value=0}
       ]}
      YCurve {keys [
        {interpMode=RCIM_Cubic, tangentMode=RCTM_Break, time=0, value=0}
        {interpMode=RCIM_Cubic, tangentMode=RCTM_Break, time=0.166, value=1.2}
        {interpMode=RCIM_Cubic, tangentMode=RCTM_Break, time=0.266, value=0.9}
        {interpMode=RCIM_Cubic, tangentMode=RCTM_Break, time=0.366, value=1.05}
        {interpMode=RCIM_Cubic, tangentMode=RCTM_Break, time=0.433, value=1}
       ]}
     }}
    Scale Sprite Size Mode=::NewEnumerator3
   )
   /SetVariables(
    Particles.MotionBlurX:NiagaraFloat=&FloatFromCurve(
     FloatCurve=DataInterface() {propertyValues {Curve {keys [
         {arriveTangentWeight=1.6667, time=6, value=0}
         <at 6.033 0.25>
         {arriveTangentWeight=0.022, time=6.066, value=0}
         <at 6.15 0.5>
         {arriveTangentWeight=0.0557, time=6.233, value=0}
         <at 6.3 1>
        ]}}}
     CurveIndex=$Particles.Age
    )
    Particles.MotionBlurY:NiagaraFloat=&FloatFromCurve(
     FloatCurve=DataInterface() {propertyValues {Curve {keys [
         {leaveTangentWeight=0.3333, time=0, value=0}
         <at 0.0804 1>
         <at 0.166 0>
         <at 0.216 0.5>
         <at 0.266 0>
         <at 0.316 0.25>
         <at 0.366 0>
         <at 0.4 0.125>
         {arriveTangentWeight=0.0223, leaveTangentWeight=0.189, time=0.433, value=0}
        ]}}}
     CurveIndex=$Particles.Age
    )
   ) {enabled=false, isSetParameters=true}
   &DynamicMaterialParameters(
    Index 0 Param 1=&FloatFromCurve(
     FloatCurve=DataInterface() {propertyValues {Curve {keys [
         {arriveTangentWeight=1.6667, time=6, value=0}
         <at 6.033 0.25>
         {arriveTangentWeight=0.022, time=6.066, value=0}
         <at 6.15 0.5>
         {arriveTangentWeight=0.0557, time=6.233, value=0}
         <at 6.3 1>
        ]}}}
     CurveIndex=$Particles.Age
     Scale Curve=20
    )
    Index 0 Param 2=&FloatFromCurve(
     FloatCurve=DataInterface() {propertyValues {Curve {keys [
         {leaveTangentWeight=0.3333, time=0, value=0}
         <at 0.0804 1>
         <at 0.166 0>
         <at 0.216 0.5>
         <at 0.266 0>
         <at 0.316 0.25>
         <at 0.366 0>
         <at 0.4 0.125>
         {arriveTangentWeight=0.0223, leaveTangentWeight=0.189, time=0.433, value=0}
        ]}}}
     CurveIndex=$Particles.Age
     Scale Curve=200
    )
    Module.Param0WriteEnabled=true
    Module.Param1WriteEnabled=true
    Module.Param2WriteEnabled=false
    Module.Param3WriteEnabled=false
   )
   &CameraOffset(Camera Offset Amount=35) {enabled=false}
  ]
  rendererClasses [&NiagaraSpriteRendererProperties]
  renderers [
   {rendererClass=&NiagaraSpriteRendererProperties}
   {rendererClass=&NiagaraSpriteRendererProperties, rendererIndex=1}
  ]
 }
}
summary {
 systemName=NS_Emote_Inferno
 emitters [
  {
   emitterName=CameraFlash
   rendererClasses [
    &NiagaraComponentRendererProperties
    &NiagaraSpriteRendererProperties
    &NiagaraLightRendererProperties
   ]
  }
  {emitterName=HeightChart, rendererClasses [&NiagaraSpriteRendererProperties]}
 ]
 userVariables []
}
